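# Spatial ambient cue for the Brimstone Cerberus head environment actor, playing a constant exhale with magic and breath layers.
# The exhale loop runs through two nested looping envelopers that slowly sway pitch and volume on different periods, avoiding static repetition.
# It mixes with a quieter whispy magic loop and randomly chosen breath one-shots that repeat after random delays.
# Concurrency allows one instance, and all quality levels share the same graph.
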SoundCue /Game/Sounds/Helios/Environment_Actors/BrimstoneCerberusHead/Spot_Cerberus_Head_Constant_Exhale_Loop_Cue
root=QualityLevel_0
nodes [
 {id=Delay_0, children [Random_0], properties {delayMax=8, delayMin=3}}
 {
  id=Enveloper_0
  children [WavePlayer_8]
  properties {
   bLoop=true
   bLoopIndefinitely=true
   loopEnd=30
   pitchCurve {editorCurveData {keys [
      {arriveTangent=-0.0013, interpMode=RCIM_Cubic, leaveTangent=-0.0013, time=0}
      <at 15 0.98 RCIM_Cubic>
      {arriveTangent=0.0013, interpMode=RCIM_Cubic, leaveTangent=0.0013, time=30}
     ]}}
   volumeCurve {editorCurveData {keys [<at 0 1>]}}
  }
 }
 {
  id=Enveloper_3
  children [Enveloper_0]
  properties {
   bLoop=true
   bLoopIndefinitely=true
   loopEnd=20
   pitchCurve {editorCurveData {keys [{time=0}]}}
   volumeCurve {editorCurveData {keys [
      {arriveTangent=-0.01, interpMode=RCIM_Cubic, time=0, value=1}
      <at 10 0.8 RCIM_Cubic>
      {interpMode=RCIM_Cubic, leaveTangent=0.01, time=20, value=1}
     ]}}
  }
 }
 {id=Looping_0, children [Delay_0]}
 {id=Mixer_1, children [Mixer_2,Looping_0], properties {inputVolume [0.8,0.5]}}
 {id=Mixer_2, children [Enveloper_3,WavePlayer_9], properties {inputVolume [0.6,0.1]}}
 {id=Modulator_1, children [Mixer_1], properties {pitchMax=1, pitchMin=1, volumeMax=1, volumeMin=1}}
 {id=QualityLevel_0, children [Modulator_1,Modulator_1,Modulator_1,Modulator_1]}
 {id=Random_0, children [WavePlayer_1,WavePlayer_2,WavePlayer_3,WavePlayer_4,WavePlayer_5,WavePlayer_7], properties {weights [1,1,1,1,1,1]}}
 {
  id=WavePlayer_1
  properties {soundWaveAssetPtr=&/Spot_Cerberus_Head_Breath_Swt_06}
 }
 {
  id=WavePlayer_2
  properties {soundWaveAssetPtr=&/Spot_Cerberus_Head_Breath_Swt_05}
 }
 {
  id=WavePlayer_3
  properties {soundWaveAssetPtr=&/Spot_Cerberus_Head_Breath_Swt_04}
 }
 {
  id=WavePlayer_4
  properties {soundWaveAssetPtr=&/Spot_Cerberus_Head_Breath_Swt_03}
 }
 {
  id=WavePlayer_5
  properties {soundWaveAssetPtr=&/Spot_Cerberus_Head_Breath_Swt_02}
 }
 {
  id=WavePlayer_7
  properties {soundWaveAssetPtr=&/Spot_Cerberus_Head_Breath_Swt_01}
 }
 {
  id=WavePlayer_8
  properties {bLooping=true, soundWaveAssetPtr=&/Spot_Cerberus_Head_Constant_Exhale_Loop}
 }
 {
  id=WavePlayer_9
  properties {bLooping=true, soundWaveAssetPtr=&/Spot_Cerberus_Head_Whispy_Magic_Fast_Loop}
 }
]
properties {
 attenuationSettings=&/Spot_Cerberus_Head_Constant_Exhale_Loop_Att
 bOverrideConcurrency=true
 maxDistance=8500
 priority=10
 soundClassObject=&Fort_Gadgets
 volumeMultiplier=0.25
 concurrencyOverrides {maxCount=1, resolutionRule=PreventNew}
}
